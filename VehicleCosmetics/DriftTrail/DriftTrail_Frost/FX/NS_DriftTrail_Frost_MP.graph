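# Niagara system for the Frost drift trail vehicle cosmetic, emitting icy effects from sockets on the vehicle's skeletal mesh.
# Two ribbons, GPU ice spikes, GPU snowflake sprites and persistent, pulsing wheel-wrap meshes make up the effect, each masked per tire by FN_DisableTireFX.
# Spawn groups pack several spike and flake variants into single emitters.
# Painted-variant modules supply colors, and a frontend-preview flag adjusts snowflake offset and velocity.
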NiagaraSystem /VehicleCosmetics/DriftTrail/DriftTrail_Frost/FX/NS_DriftTrail_Frost_MP
dependencies {
 usedDataInterfaces [
  &NiagaraDataInterfaceCurve
  &NiagaraDataInterfaceSpriteRendererInfo
  &NiagaraDataInterfaceSkeletalMesh
 ]
 usedDynamicInputs []
 usedModules [
  &self:SystemSpawnScript.SystemScriptSource.SystemScriptGraph.NiagaraNodeAssignment_0.SetVariables_#1
  &SystemState
  &FN_UnpackVehicleFlags
  /Game/Effects/Niagara/Modules/Vehicle/FN_UpdateVehicleAttribs
  /Game/Effects/Niagara/Modules/Vehicle/FN_VehicleFlags_Cheatcodes
 ]
}
emitters {
 FrostRibbon {
  emitterUpdateScript [
   &EmitterState()
   &VC_RateQualityHandler(Epic Spawn Rate=90, High Spawn Rate=60, Low Spawn Rate=30, Medium Spawn Rate=45, Use Custom Rate?=true)
   &VC_SpawnRibbon_v1(Number of Ribbons=2)
  ]
  particleSpawnScript [
   &VC_PaintedVariantsOverride(BurntSienna=<rgba 0.3281 0.1145 0 1>, DefaultColor=<rgba 0.5729 0.8986 1 1>)
   &VC_InitializeRibbon_v1(
    Custom Offset=true
    Enable Skeletal Position=true
    Lifetime=1.15
    Local Vector Offset=&MakeCustomVectorFromBool(Bool=true, True Vector=<-60 0 -30>, False Vector=<-60 0 -20>)
    Offset Amount=1
    Ribbon Width=30
   )
   &AddVelocity(Velocity=<-1000 0 0>) {module=AddVelocity001}
   &Color(Color=$User.PaintedColor) {enabled=false, module=Color001}
  ]
  particleUpdateScript [
   &FN_DisableTireFX(Mask Index Attribute:NiagaraInt32=$Particles.SpawnGroup, Mask Index FL=2, Mask Index FR=3, Mask Index RL=0, Mask Index RR=1)
   &ParticleState()
   &InheritSourceMovement(
    Applied Position Scale=&Multiply_VectorByFloat(
     Vector=<1 1 1>
     Float=&FloatFromCurve(FloatCurve=DataInterface() {propertyValues {Curve {keys [<at 0.2 1 RCIM_Cubic>, {arriveTangent=-1.6761, interpMode=RCIM_Cubic, leaveTangent=-1.6761, tangentMode=RCTM_User, time=1, value=0}]}}})
    )
    ApplyForce=false
    ApplyPosition=true
    ApplyPositionFalloff=false
   )
   &SolveForcesAndVelocity(Clamp Velocity=false, Limit Acceleration=false)
   &VC_UpdateRibbon_v1(Enable Skeletal Position=true)
   &ScaleColor(
    Scale Alpha=&FloatFromCurve(FloatCurve=DataInterface() {propertyValues {Curve {keys [<at 0 0>,<at 0.4 1>,<at 0.85 0>]}}}, Scale Curve=0.5)
    ScaleA=true
    ScaleRGB=true
   )
   &ScaleRibbonWidth(Ribbon Width Scale=&FloatFromCurve(FloatCurve=DataInterface() {propertyValues {Curve {keys [<at 0 1>,<at 0.4 3>,<at 1 8>]}}}))
   &DynamicMaterialParameters(
    Index 0 Param 1=1
    Index 0 Param 4=1.5
    Index 1 Param 2=1.5
    Module.Index0Param0Write=true
    Module.Index0Param1Write=true
    Module.Index0Param2Write=true
    Module.Index0Param3Write=true
    Module.Index1Param0Write=true
    Module.Index1Param1Write=true
    Module.Index1Param2Write=true
    Module.Index1Param3Write=true
    Write Parameter Index 1=true
   )
  ]
  rendererClasses [&NiagaraRibbonRendererProperties]
  renderers [{rendererClass=&NiagaraRibbonRendererProperties}]
 }
 IceSpikes {
  simTarget=GPUComputeSim
  emitterUpdateScript [
   &EmitterState()
   &SpawnRate(
    SpawnRate=&MakeCustomFloatFromBool(Bool=$User.bIsFrontendPreview, True Float=40, False Float=40)
    Use Spawn Probability=true
   )
   &SpawnRate(
    SpawnGroup=1
    SpawnRate=&MakeCustomFloatFromBool(Bool=$User.bIsFrontendPreview, True Float=20, False Float=20)
    Use Spawn Probability=true
   ) {module=SpawnRate004}
   &SpawnRate(
    SpawnGroup=2
    SpawnRate=&MakeCustomFloatFromBool(Bool=$User.bIsFrontendPreview, True Float=40, False Float=40)
    Use Spawn Probability=true
   ) {module=SpawnRate005}
  ]
  particleSpawnScript [
   &VC_PaintedVariantsOverride(
    BurntSienna=<rgba 0.3281 0.1145 0 1>
    DefaultColor=<rgba 0.2552 0.8231 1 1>
    Gold:LinearColor=<rgba 0.3594 0.2277 0.0373 1>
    TitaniumWhite=<rgba 0.4381 0.6087 0.75 1>
   )
   &InitializeParticle(
    Color Mode=::NewEnumerator0
    Lifetime=/VehicleCosmetics/FX/Niagara/Modules/MergeEmitters/VC_FloatBySpawnGroup(
     SpawnGroup or ExecIndex:NiagaraInt32=0
     CustomInt:NiagaraInt32=0
     MaximumSpawnGroups:NiagaraInt32=3
     Float SG 0:NiagaraFloat=&RandomRangeFloat(Minimum=0.4, Maximum=0.6)
     Float SG 1:NiagaraFloat=&RandomRangeFloat(Minimum=0.4, Maximum=0.4)
     Float SG 2:NiagaraFloat=&RandomRangeFloat(Minimum=0.2, Maximum=0.2)
     Float SG 3:NiagaraFloat=&RandomRangeFloat(Minimum=0.4, Maximum=1)
     Float SG 4:NiagaraFloat=0
     Float SG 5:NiagaraFloat=0
    )
    Mesh Scale=/VehicleCosmetics/FX/Niagara/Modules/MergeEmitters/VC_VectorBySpawnGroup(
     SpawnGroup or ExecIndex:NiagaraInt32=0
     CustomInt:NiagaraInt32=0
     MaximumSpawnGroups:NiagaraInt32=3
     Vector Group 0:Vector3f=&RandomRangeVector(Minimum=<0.6 1.2 1.2>, Maximum=<1 2.4 2.8>)
     Vector Group 1:Vector3f=&RandomRangeVector(Minimum=<0.6 1.2 1.2>, Maximum=<1 0.2 2.8>)
     Vector Group 2:Vector3f=&RandomRangeVector(Minimum=<0.6 1.2 1.2>, Maximum=<1 2.4 2.8>)
     Vector Group 3:Vector3f=&RandomRangeVector(
      Minimum=&MakeCustomVectorFromBool(Bool=$User.bIsFrontendPreview, True Vector=<-100 0 0>, False Vector=<-500 0 0>)
      Maximum=&MakeCustomVectorFromBool(Bool=$User.bIsFrontendPreview, True Vector=<-500 0 0>, False Vector=<-1500 0 0>)
     )
     Vector Group 4:Vector3f=<0 0 0>
     Vector Group 5:Vector3f=<0 0 0>
    )
    Mesh Scale Mode=::NewEnumerator4
    UsePositionOffset=false
    Write Material Random=true
   )
   &PartitionParticles()
   &SkeletalMeshLocation(
    Bone / Socket Index=&Add_Integer(A=$Particles.PartitionParticles.Partition, B=4)
    Bone / Socket Sampling Mode=::NewEnumerator5
    MaskSpawnGroup=false
    Mesh Sampling Type=::NewEnumerator6
    Orientation Sampling=::NewEnumerator1
    Sampled Position Offset=&RandomRangeVector(Minimum=<-60 -5 -40>, Maximum=<-60 -5 -40>)
    Sampled Position Offset Coordinate Space:ENiagaraSkelMeshTransforms=::NewEnumerator2
    Skeletal Mesh=$System.SkeletalMesh
   )
   &AddVelocity(Velocity=&VC_VectorBySpawnGroup(
     SpawnGroup or ExecIndex:NiagaraInt32=0
     CustomInt:NiagaraInt32=0
     MaximumSpawnGroups:NiagaraInt32=3
     Vector Group 0:Vector3f=&MakeCustomVectorFromBool(Bool=$User.bIsFrontendPreview, True Vector=<-200 0 0>, False Vector=<-200 0 0>)
     Vector Group 1:Vector3f=&MakeCustomVectorFromBool(Bool=$User.bIsFrontendPreview, True Vector=<-70 0 0>, False Vector=<-70 0 0>)
     Vector Group 2:Vector3f=&MakeCustomVectorFromBool(Bool=$User.bIsFrontendPreview, True Vector=<-10 0 0>, False Vector=<-10 0 0>)
     Vector Group 3:Vector3f=&RandomRangeVector(
      Minimum=&MakeCustomVectorFromBool(Bool=$User.bIsFrontendPreview, True Vector=<-100 0 0>, False Vector=<-500 0 0>)
      Maximum=&MakeCustomVectorFromBool(Bool=$User.bIsFrontendPreview, True Vector=<-500 0 0>, False Vector=<-1500 0 0>)
     )
     Vector Group 4:Vector3f=<0 0 0>
     Vector Group 5:Vector3f=<0 0 0>
    ))
   &Color(Color=$User.PaintedColor) {enabled=false}
   &Color(Color=<rgba 0.35 0.8456 1 1>) {enabled=false, module=Color001}
  ]
  particleUpdateScript [
   &FN_DisableTireFX(
    Mask Index Attribute:NiagaraInt32=$Particles.PartitionParticles.Partition
    Mask Index FL=2
    Mask Index FR=3
    Mask Index RL=0
    Mask Index RR=1
   )
   &ParticleState()
   &ScaleMeshSize(Scale Factor=&VC_VectorBySpawnGroup(
     SpawnGroup or ExecIndex:NiagaraInt32=0
     CustomInt:NiagaraInt32=0
     MaximumSpawnGroups:NiagaraInt32=3
     Vector Group 0:Vector3f=&MakeVector(
      X=&FloatFromCurve(FloatCurve=DataInterface() {propertyValues {Curve {defaultValue=0.7, keys [{tangentMode=RCTM_User, time=0, value=0.7}, {tangentMode=RCTM_User, time=0.1, value=0.9875}]}}})
      Y=1
      Z=&FloatFromCurve(
       FloatCurve=DataInterface() {propertyValues {Curve {
          defaultValue=0.7
          keys [
           {interpMode=RCIM_Cubic, tangentMode=RCTM_User, time=0, value=0.55}
           {arriveTangent=-1.9526, interpMode=RCIM_Cubic, leaveTangent=-1.9526, tangentMode=RCTM_User, time=0.3, value=1}
           {tangentMode=RCTM_User, time=1, value=0.3}
          ]
         }}}
       Scale Curve=&RandomRangeFloat(Minimum=0.6, Maximum=1.4)
      )
     )
     Vector Group 1:Vector3f=&MakeVector(
      X=&FloatFromCurve(FloatCurve=DataInterface() {propertyValues {Curve {defaultValue=0.7, keys [{tangentMode=RCTM_User, time=0, value=0.7}, {tangentMode=RCTM_User, time=0.1, value=0.9875}]}}})
      Y=1
      Z=&FloatFromCurve(
       FloatCurve=DataInterface() {propertyValues {Curve {
          defaultValue=0.7
          keys [
           {interpMode=RCIM_Cubic, tangentMode=RCTM_User, time=0, value=0.55}
           {interpMode=RCIM_Cubic, tangentMode=RCTM_User, time=0.5, value=1}
           {tangentMode=RCTM_User, time=1, value=0.65}
          ]
         }}}
       Scale Curve=&RandomRangeFloat(Minimum=0.6, Maximum=1.4)
      )
     )
     Vector Group 2:Vector3f=&MakeVector(
      X=1
      Y=1
      Z=&FloatFromCurve(
       FloatCurve=DataInterface() {propertyValues {Curve {
          defaultValue=0.7
          keys [
           {interpMode=RCIM_Cubic, tangentMode=RCTM_User, time=0, value=0.55}
           {interpMode=RCIM_Cubic, tangentMode=RCTM_User, time=0.5, value=1}
           {tangentMode=RCTM_User, time=1, value=0}
          ]
         }}}
       Scale Curve=&RandomRangeFloat(Minimum=1.8, Maximum=2)
      )
     )
     Vector Group 3:Vector3f=&RandomRangeVector(
      Minimum=&MakeCustomVectorFromBool(Bool=$User.bIsFrontendPreview, True Vector=<-100 0 0>, False Vector=<-500 0 0>)
      Maximum=&MakeCustomVectorFromBool(Bool=$User.bIsFrontendPreview, True Vector=<-500 0 0>, False Vector=<-1500 0 0>)
     )
     Vector Group 4:Vector3f=<0 0 0>
     Vector Group 5:Vector3f=<0 0 0>
    ))
   &UpdateMeshOrientation(
    Facing Direction=&Lerp_Vector(A=<1 0 0.2>, B=&MakeVector(X=0, Y=0, Z=-0.1), Alpha=$Particles.NormalizedAge)
    Facing Interpolation=false
    Orientation Method=::NewEnumerator1
    Override Quat=false
    TransformFacing=false
    TransformUp=false
   )
   &DynamicMaterialParameters(
    Index 0 Param 1=&RandomRangeFloat(Minimum=4, Maximum=6)
    Module.Index0Param0Write=true
    Module.Index0Param1Write=true
    Module.Index0Param2Write=true
    Module.Index0Param3Write=true
   )
   &SolveForcesAndVelocity(Clamp Velocity=false, Limit Acceleration=false)
  ]
  rendererClasses [&NiagaraMeshRendererProperties, &NiagaraSpriteRendererProperties]
  renderers [
   {rendererClass=&NiagaraMeshRendererProperties, rendererIndex=0}
   {rendererClass=&NiagaraSpriteRendererProperties, rendererIndex=1}
  ]
 }
 Snowflakes {
  simTarget=GPUComputeSim
  emitterUpdateScript [
   &EmitterState()
   &SpawnRate(SpawnRate=40, Use Spawn Probability=false)
   &SpawnRate(SpawnGroup=1, SpawnRate=90, Use Spawn Probability=false) {module=SpawnRate001}
  ]
  particleSpawnScript [
   &VC_PaintedVariantsOverride(
    Black=<rgba 0.03 0.03 0.03 0.8>
    BurntSienna=<rgba 0.78 0.3978 0.0156 0.9>
    Cobalt=<rgba 0.0547 0.1782 1.1 0.7>
    Crimson=<rgba 0.9 0 0.045 0.5>
    DefaultColor=<rgba 1.1667 1.4971 1.6 0.45>
    ForestGreen=<rgba 0 1.1 0.0556 0.6>
    Grey=<rgba 0.59 0.5439 0.5439 0.8>
    Lime=<rgba 0.8 1.2 0.24 0.6>
    Orange=<rgba 1 0.6 0 0.8>
    Pink=<rgba 1 0.32 0.7733 0.5>
    Purple=<rgba 0.624 0.18 0.9 0.8>
    R1 Black:LinearColor=<rgba 0.05 0.05 0.05 2>
    R1 BurntSienna:LinearColor=<rgba 1 0.25 0 2>
    R1 Cobalt:LinearColor=<rgba 0 0.3333 1 2>
    R1 Crimson:LinearColor=<rgba 2 0 0.0542 1>
    R1 DefaultColor=<rgba 2 2 2 2>
    R1 ForestGreen:LinearColor=<rgba 0 2 0.1011 1>
    R1 Grey:LinearColor=<rgba 0.7 0.7 0.7 1>
    R1 Lime:LinearColor=<rgba 1 2 0 1>
    R1 Orange:LinearColor=<rgba 2 0.6976 0 2>
    R1 Pink:LinearColor=<rgba 2 0.2396 1.3598 2>
    R1 Purple:LinearColor=<rgba 0.8208 0 2 2>
    R1 Saffron:LinearColor=<rgba 2 1.5333 0 2>
    R1 SkyBlue:LinearColor=<rgba 0 1.3308 2 2>
    R1 TitaniumWhite:LinearColor=<rgba 2 2 2 2>
    Saffron=<rgba 1.4 1.19 0.14 0.5>
    SkyBlue=<rgba 0.285 0.7275 0.95 1>
    TitaniumWhite=<rgba 1 1 1 1>
    Use R1 Painted Overrides=true
   )
   &InitializeParticle(
    Color Mode=::NewEnumerator0
    Lifetime=&VC_FloatBySpawnGroup(
     SpawnGroup or ExecIndex:NiagaraInt32=0
     CustomInt:NiagaraInt32=0
     MaximumSpawnGroups:NiagaraInt32=2
     Float SG 0:NiagaraFloat=&RandomRangeFloat(Minimum=1.2, Maximum=1.6)
     Float SG 1:NiagaraFloat=&RandomRangeFloat(Minimum=0.6, Maximum=1)
     Float SG 2:NiagaraFloat=&RandomRangeFloat(Minimum=0.2, Maximum=0.2)
     Float SG 3:NiagaraFloat=&RandomRangeFloat(Minimum=0.4, Maximum=1)
     Float SG 4:NiagaraFloat=0
     Float SG 5:NiagaraFloat=0
    )
    Sprite Rotation Mode=::NewEnumerator2
    Sprite Size Mode=::NewEnumerator3
    Uniform Sprite Size=&VC_FloatBySpawnGroup(
     SpawnGroup or ExecIndex:NiagaraInt32=0
     CustomInt:NiagaraInt32=0
     MaximumSpawnGroups:NiagaraInt32=2
     Float SG 0:NiagaraFloat=&RandomRangeFloat(Minimum=1, Maximum=3)
     Float SG 1:NiagaraFloat=&RandomRangeFloat(Minimum=70, Maximum=100)
     Float SG 2:NiagaraFloat=&RandomRangeFloat(Minimum=0.2, Maximum=0.2)
     Float SG 3:NiagaraFloat=&RandomRangeFloat(Minimum=0.4, Maximum=1)
     Float SG 4:NiagaraFloat=0
     Float SG 5:NiagaraFloat=0
    )
    UsePositionOffset=false
    Write Material Random=true
   )
   &PartitionParticles()
   /SetVariables(Particles.RenderVisTag:NiagaraInt32=&SpawnGroup()) {isSetParameters=true}
   &SkeletalMeshLocation(
    Bone / Socket Index=&Add_Integer(A=$Particles.PartitionParticles.Partition, B=4)
    Bone / Socket Sampling Mode=::NewEnumerator5
    Interpolate Mesh Sampling=false
    MaskSpawnGroup=false
    Mesh Sampling Type=::NewEnumerator6
    Sampled Position Offset=&MakeCustomVectorFromBool(Bool=$User.bIsFrontendPreview, True Vector=<-15 0 10>, False Vector=<0 0 0>)
    Skeletal Mesh=$System.SkeletalMesh
   )
   &AddVelocity(Velocity=&VC_VectorBySpawnGroup(
     SpawnGroup or ExecIndex:NiagaraInt32=0
     CustomInt:NiagaraInt32=0
     MaximumSpawnGroups:NiagaraInt32=2
     Vector Group 0:Vector3f=&MakeCustomVectorFromBool(
      Bool=$User.bIsFrontendPreview
      True Vector=&RandomRangeVector(Minimum=<-600 0 90>, Maximum=<-400 0 110>)
      False Vector=&RandomRangeVector(Minimum=<-400 0 20>, Maximum=<-200 0 100>)
     )
     Vector Group 1:Vector3f=&MakeCustomVectorFromBool(
      Bool=$User.bIsFrontendPreview
      True Vector=&RandomRangeVector(Minimum=<-400 0 40>, Maximum=<-600 0 70>)
      False Vector=&RandomRangeVector(Minimum=<-600 0 -10>, Maximum=<-800 0 0>)
     )
     Vector Group 2:Vector3f=<0 0 0>
     Vector Group 3:Vector3f=<0 0 0>
     Vector Group 4:Vector3f=<0 0 0>
     Vector Group 5:Vector3f=<0 0 0>
    ))
   &ShapeLocation(
    MaskSpawnGroup=false
    Offset=<-80 0 -20>
    Offset Coordinate Space=::Local
    Offset Mode=::NewEnumerator0
    Override Seed=false
    Sphere Radius=&RandomRangeFloat(Minimum=30, Maximum=60)
    UseOwnerScale=true
   )
   &Color(Color=&MakeCustomLinearColorFromBool(
     Bool=&SetBoolByIntComparison(A=$Particles.RenderVisTag, B=0, Comparison Type=::NewEnumerator0)
     True Color=$Particles.Color
     False Color=&Multiply_LinearColorByFloat(Linear Color=$Particles.Color, Float=3)
    ))
  ]
  particleUpdateScript [
   &FN_DisableTireFX(
    Mask Index Attribute:NiagaraInt32=$Particles.PartitionParticles.Partition
    Mask Index FL=2
    Mask Index FR=3
    Mask Index RL=0
    Mask Index RR=1
   )
   &InheritSourceMovement(
    Applied Force Scale=<0.15 0.15 0.15>
    Applied Position Scale=&Multiply_VectorByFloat(
     Vector=<1 1 1>
     Float=&FloatFromCurve(FloatCurve=DataInterface() {propertyValues {Curve {keys [
          {
           arriveTangent=-1.1883
           arriveTangentWeight=0.6304
           interpMode=RCIM_Cubic
           leaveTangent=-1.1883
           leaveTangentWeight=0.6304
           tangentMode=RCTM_User
           tangentWeightMode=RCTWM_WeightedBoth
           time=0.5
           value=1
          }
          {arriveTangent=-3.2, interpMode=RCIM_Cubic, leaveTangent=-3.2, tangentMode=RCTM_User, time=1, value=0}
         ]}}})
    )
    ApplyForce=false
    ApplyPosition=true
    ApplyPositionFalloff=false
   )
   &ParticleState()
   &ScaleSpriteSize(
    Scale Sprite Size Mode=::NewEnumerator0
    Uniform Scale Factor=&VC_FloatBySpawnGroup(
     SpawnGroup or ExecIndex:NiagaraInt32=0
     CustomInt:NiagaraInt32=0
     MaximumSpawnGroups:NiagaraInt32=2
     Float SG 0:NiagaraFloat=&FloatFromCurve(FloatCurve=DataInterface() {propertyValues {Curve {keys [{tangentMode=RCTM_User, time=0, value=0}, {tangentMode=RCTM_User, time=0.2, value=1}]}}})
     Float SG 1:NiagaraFloat=1
     Float SG 2:NiagaraFloat=&RandomRangeFloat(Minimum=0.2, Maximum=0.2)
     Float SG 3:NiagaraFloat=&RandomRangeFloat(Minimum=0.4, Maximum=1)
     Float SG 4:NiagaraFloat=0
     Float SG 5:NiagaraFloat=0
    )
   ) {module=ScaleSpriteSize002}
   &ScaleSpriteSize(
    Scale Sprite Size Mode=::NewEnumerator0
    Uniform Scale Factor=&VC_FloatBySpawnGroup(
     SpawnGroup or ExecIndex:NiagaraInt32=0
     CustomInt:NiagaraInt32=0
     MaximumSpawnGroups:NiagaraInt32=2
     Float SG 0:NiagaraFloat=&FloatFromCurve(FloatCurve=DataInterface() {propertyValues {Curve {keys [{interpMode=RCIM_Cubic, tangentMode=RCTM_Break, time=0, value=7}, {interpMode=RCIM_Cubic, tangentMode=RCTM_Break, time=1, value=1}]}}})
     Float SG 1:NiagaraFloat=1.5
     Float SG 2:NiagaraFloat=&RandomRangeFloat(Minimum=0.2, Maximum=0.2)
     Float SG 3:NiagaraFloat=&RandomRangeFloat(Minimum=0.4, Maximum=1)
     Float SG 4:NiagaraFloat=0
     Float SG 5:NiagaraFloat=0
    )
   ) {module=ScaleSpriteSize001}
   &DynamicMaterialParameters(
    Index 0 Param 1=1
    Module.Index0Param0Write=true
    Module.Index0Param1Write=true
    Module.Index0Param2Write=true
    Module.Index0Param3Write=true
   )
   &SolveForcesAndVelocity(Clamp Velocity=false, Limit Acceleration=false)
   &SpriteFacingAndAlignment(
    Sprite Facing=&RandomRangeVector(Minimum=<0 0 0>, Maximum=<1 1 1>)
    WriteAlignment=false
    WriteFacing=true
   )
   &SubUVAnimation(
    Evaluation Type=::NewEnumerator0
    "Sprite Renderer "=DataInterface() {propertyValues {SpriteRenderer=&self:Snowflakes_0.NiagaraSpriteRendererProperties_1}}
    SubUV Animation Mode:ENiagara_SubUVLookupModeV2=::NewEnumerator2
    UseEndFrame=false
    UseStartFrame=false
   )
  ]
  rendererClasses [&NiagaraSpriteRendererProperties]
  renderers [
   {rendererClass=&NiagaraSpriteRendererProperties}
   {rendererClass=&NiagaraSpriteRendererProperties, rendererIndex=1}
   {rendererClass=&NiagaraSpriteRendererProperties, rendererIndex=2}
  ]
 }
 WheelWrap001 {
  emitterUpdateScript [
   &EmitterState(
    Inactive Response=::NewEnumerator1
    Life Cycle Mode=::NewEnumerator1
    Loop Behavior=::NewEnumerator1
    Loop Duration Mode=::NewEnumerator1
    UseLoopDelay=false
   )
   &SpawnBurst_Instantaneous(Spawn Count=2, Use Spawn Probability=false, UseLoopCountLimit=false)
  ]
  particleSpawnScript [
   &VC_PaintedVariantsOverride(BurntSienna=<rgba 0.3281 0.1145 0 1>, DefaultColor=<rgba 0.5729 0.8986 1 1>)
   &InitializeParticle(
    Color Mode=::NewEnumerator0
    Lifetime=5
    Mesh Scale Mode=::NewEnumerator4
    Sprite Rotation Mode=::NewEnumerator2
    UsePositionOffset=false
    Write Material Random=true
   )
   &PartitionParticles()
   &SkeletalMeshLocation(
    Bone / Socket Index=&Add_Integer(A=$Particles.PartitionParticles.Partition, B=4)
    Bone / Socket Sampling Mode=::NewEnumerator5
    Interpolate Mesh Sampling=false
    MaskSpawnGroup=false
    Mesh Sampling Type=::NewEnumerator6
    Skeletal Mesh=$System.SkeletalMesh
   )
   &Color(Color=$User.PaintedColor) {enabled=false, module=Color001}
   &InitialMeshOrientation(Mesh Orientation Mode=::NewEnumerator7, Rotation=<0 0 0.5>, Use Rotation Vector=true)
   &Color(Color=<rgba 0 1 0 1>)
  ]
  particleUpdateScript [
   &FN_DisableTireFX(
    Mask Index Attribute:NiagaraInt32=$Particles.PartitionParticles.Partition
    Mask Index FL=2
    Mask Index FR=3
    Mask Index RL=0
    Mask Index RR=1
   )
   &VC_WheelWrapScale(Wheel Index Attribute:NiagaraInt32=&Add_Integer(A=$Particles.PartitionParticles.Partition, B=2))
   &DynamicMaterialParameters(
    Index 0 Param 1=1
    Index 0 Param 2=1
    Index 0 Param 3=1
    Index 0 Param 4=1
    Module.Index0Param0Write=true
    Module.Index0Param1Write=true
    Module.Index0Param2Write=true
    Module.Index0Param3Write=true
   )
   &ParticleState(Kill Particles When Lifetime Has Elapsed=false, Loop Particles Lifetime=true)
   &UpdateMeshOrientation(Rotation Coordinate Space=::NewEnumerator2, Rotation Rate=5) {enabled=false}
   &SkeletalMeshLocation(
    Bone / Socket Index=&Add_Integer(A=$Particles.PartitionParticles.Partition, B=4)
    Bone / Socket Sampling Mode=::NewEnumerator5
    MaskSpawnGroup=false
    Mesh Sampling Type=::NewEnumerator6
    Skeletal Mesh=$System.SkeletalMesh
   ) {module=SkeletalMeshLocation001}
   &SolveForcesAndVelocity(Clamp Velocity=false, Limit Acceleration=false)
   &ScaleColor(
    Scale RGB=&VectorFromFloat(Value=&Waveform("Waveform [1]"=::NewEnumerator0, "[1] Phase Scale"=5, "[1] Frequency"=0.5, "[1] Amplitude Min/Max"=<50 100>))
    ScaleA=true
    ScaleRGB=true
   )
  ]
  rendererClasses [&NiagaraMeshRendererProperties]
  renderers [{rendererClass=&NiagaraMeshRendererProperties, rendererIndex=0}]
 }
}
summary {
 systemName=NS_DriftTrail_Frost_MP
 emitters [
  {emitterName=Snowflakes, simTarget=GPUComputeSim, rendererClasses [&NiagaraSpriteRendererProperties]}
  {
   emitterName=IceSpikes
   simTarget=GPUComputeSim
   rendererClasses [&NiagaraMeshRendererProperties, &NiagaraSpriteRendererProperties]
  }
  {emitterName=FrostRibbon, rendererClasses [&NiagaraRibbonRendererProperties]}
  {emitterName=WheelWrap001, rendererClasses [&NiagaraMeshRendererProperties]}
 ]
 userVariables [
  User.bIsFrontendPreview {type=NiagaraBool}
  User.bIsPainted {type=NiagaraBool}
  User.ColorVariantID {type=NiagaraInt32}
  User.PaintedColor {type=LinearColor, description="", defaultValue {struct=LinearColor, value=<rgba 1 1 1 1>}}
  User.VehicleFlags {type=NiagaraInt32, defaultValue=806289408}
  User.Wheel_Radius_Array {
   type=NiagaraDataInterfaceArrayFloat
   defaultValue {
    struct=NiagaraExt_VariableValue_DataInterface
    value {
     dataInterface=&self:NiagaraDataInterfaceArrayFloat_0
     dataInterfaceClass=&NiagaraDataInterfaceArrayFloat
    }
   }
  }
  User.Wheel_Width_Array {
   type=NiagaraDataInterfaceArrayFloat
   defaultValue {
    struct=NiagaraExt_VariableValue_DataInterface
    value {
     dataInterface=&self:NiagaraDataInterfaceArrayFloat_1
     dataInterfaceClass=&NiagaraDataInterfaceArrayFloat
    }
   }
  }
  User.WheelSize {type=NiagaraFloat, defaultValue=0, description=""}
 ]
}
